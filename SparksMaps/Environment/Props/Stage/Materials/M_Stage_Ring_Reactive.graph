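# Surface material for a reactive stage ring prop, with glow tinted by stage colors and a beat pulse function.
# A softened normal map bends a tangent-space reflection vector, which offsets emissive UVs to fake depth.
# The offset blue-channel sample is masked by an unshifted one, tinted by MF_MojoStageColors Tertiary, then fed to MF_MojoBeatPulse.
# The StageColor lerp has Tertiary on both inputs, so that parameter has no effect.
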Material /SparksMaps/Environment/Props/Stage/Materials/M_Stage_Ring_Reactive
graph {
 properties {pixelDepthOffsetMode=PDOM_Legacy, usedShadingModels=Default Lit}
 nodes [
  n1 = TextureSampleParameter2D() {parameterName=Diffuse, texture=/Game/Environments/Galileo/Textures/T_Galileo_Stage_Cylinder_D}
  n2 = ScalarParameter() {defaultValue=1, parameterName=UVTile_Depth}
  n3 = TextureCoordinate()
  n4 = TextureSampleParameter2D() {parameterName=Normal, samplerType=SAMPLERTYPE_Normal, texture=/Game/Environments/Galileo/Textures/T_Galileo_Stage_Cylinder_N}
  n5 = MaterialFunctionCall!(
   Normal=MaterialFunctionCall(R=Multiply(A=@n4.R) {constB=0.1}, G=Multiply(A=@n4.G) {constB=0.1}, B=Constant() {r=1}) {materialFunction=&AppendMany}.RGB
   CameraVector=Transform(None=CameraVectorWS()) {transformSourceType=TRANSFORMSOURCE_World, transformType=TRANSFORM_Tangent}
  ) {materialFunction=/Engine/Functions/Engine_MaterialFunctions02/Utility/CustomReflectionVector}
  n6 = ComponentMask(None=@n5.Reflection Vector) {b=true}
  n7 = Multiply(
   A=ComponentMask(None=@n5.Reflection Vector) {g=true, r=true}
   B=Divide(A=ScalarParameter() {defaultValue=6, parameterName=Depth}, B=Abs(None=@n6))
  )
  n8 = Divide(A=Constant() {r=1}, B=Constant() {r=1024})
  n9 = TextureObjectParameter() {parameterName=EmissiveTexture, texture=/Game/Environments/Galileo/Textures/T_Galileo_Stage_Cylinder_E}
  n10 = TextureSample(UVs=Add(A=Multiply(A=@n2, B=@n3), B=Multiply(A=@n7, B=@n8)), Tex=@n9) {texture=&T_Galileo_Stage_Cylinder_E}
  n11 = TextureSample(Tex=@n9) {texture=&T_Galileo_Stage_Cylinder_E}
  n12 = MaterialFunctionCall!(CustomUPosition) {materialFunction=/SparksStageShow/Systems/StageManager/FXManager/MaterialFunctions/MF_MojoStageColors}
  n13 = Multiply(
   A=Multiply(A=Constant() {r=3}, B=Multiply(A=@n10.B, B=@n11.B))
   B=LinearInterpolate(A=@n12.Tertiary, B=@n12.Tertiary, Alpha=ScalarParameter() {parameterName=StageColor}) {constAlpha=0}
  )
  Add(A=@n1.RGB, B=@n13)
  MaterialFunctionCall!(Emissive=@n13, BeatPulse Strength, BeatPulse Min, BeatPulse Max) {materialFunction=/SparksMaps/Environment/MaterialFunctions/MF_MojoBeatPulse}
  Reroute(None=@n4.RGB)
  Saturate(None=LinearInterpolate(A=@n1.RGB, B=Constant() {r=1}, Alpha=@n11.B))
  TextureSampleParameter2D() {parameterName=Spec/Metal/Rough, samplerType=SAMPLERTYPE_Masks, texture=/Game/Environments/Galileo/Textures/T_Galileo_Stage_Cylinder_S}
 ]
}
